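# Sound cue for the machine pistol's equip foley.
# A single MachinePistol_Equip wave player feeds a Modulator left at its default settings.
# The cue lowers volume, raises pitch, routes to the weapon foley sound class and borrows the weapon reload attenuation.
# It overrides concurrency and has its debug flag left on.
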SoundCue /WeaponsUnvaulted/Assets/Audio/MachinePistol/SC_MachinePistol_Equip
root=Modulator_0
nodes [
 {id=Modulator_0, children [WavePlayer_0]}
 {id=WavePlayer_0, properties {soundWaveAssetPtr=/WeaponsUnvaulted/Assets/Audio/MachinePistol/SoundWaves/MachinePistol_Equip}}
]
properties {
 attenuationSettings=&Fort_Weapon_Reload_Att
 bDebug=true
 bOverrideConcurrency=true
 maxDistance=2250
 pitchMultiplier=1.2
 soundClassObject=&Fort_WeaponFoley
 volumeMultiplier=0.45
}
